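# Idle glow effect for the Outershell mic stand cosmetic, built as a Niagara system.
# CapsuleGlow bursts three or four persistent blue particles, depending on User.IsInstrument, using sprite and mesh renderers.
# Partition index picks each fixed position, size, tint and camera offset: three small low glows plus a larger high one.
# System time-of-day color and a front-end alpha modifier scale color, owner scale drives sprite size, and the Light emitter is disabled.
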NiagaraSystem /SparksCosmetics/Instrument/MicStand/Outershell/Effects/NS_MicStand_Outershell_Idle_FX
dependencies {
 usedDataInterfaces [&NiagaraDataInterfaceCurve, &NiagaraDataInterfaceSocketReader]
 usedDynamicInputs [
  &MergeHelper_Vector2Switch
  &FNTimeOfDayColor
  &MakeFloatFromVector
 ]
 usedModules [
  &self:SystemSpawnScript.SystemScriptSource.SystemScriptGraph.NiagaraNodeAssignment_11.SetVariables_#1
  &SystemState
  &self:SystemSpawnScript.SystemScriptSource.SystemScriptGraph.NiagaraNodeAssignment_10.SetVariables_#2
  &self:SystemSpawnScript.SystemScriptSource.SystemScriptGraph.NiagaraNodeAssignment_2.SetVariables_#3
 ]
}
emitters {
 CapsuleGlow {
  emitterUpdateScript [
   &EmitterState(Life Cycle Mode=::NewEnumerator1, Loop Behavior=::NewEnumerator1, Loop Duration Mode=::NewEnumerator1, UseLoopDelay=false)
   &SpawnBurst_Instantaneous(
    Spawn Count=&MakeCustomIntFromBool(Bool=&SetBoolByFloatComparison(A=$User.IsInstrument, B=0), True INT=3, False INT=4)
    Use Spawn Probability=false
    UseLoopCountLimit=true
   )
  ]
  particleSpawnScript [
   &PartitionParticles(ID to Partition:NiagaraInt32=&ReturnExecIndex(), Number Of Partitions=4)
   &InitializeParticle(
    Color=&MergeHelper_ColorSwitch(
     0=<rgba 0.0286 0.5495 1 0.15>
     1=<rgba 0.0286 0.5495 1 0.15>
     2=<rgba 0.0286 0.5495 1 0.15>
     3=<rgba 0 0.5362 1 0.25>
     Index=$Particles.PartitionParticles.Partition
     Num=::NewEnumerator2
    )
    Position=&MergeHelper_VectorSwitch(
     0=<-2.97 2.008 18.25>
     1=<0 -2.7 18.25>
     2=<2.55 1.7 18.25>
     3=<0 -12.7 144>
     Index=$Particles.PartitionParticles.Partition
     Num=::NewEnumerator2
    )
    Position Mode=::NewEnumerator1
    Sprite Size Mode=::NewEnumerator3
    Uniform Sprite Size=&MergeHelper_FloatSwitch(0=4, 1=4, 2=4, 3=11, Index=$Particles.PartitionParticles.Partition, Num=::NewEnumerator2)
    UsePositionOffset=false
    Write Material Random=true
   )
   &CameraOffset(Camera Offset Amount=&MergeHelper_FloatSwitch(0=3, 1=3, 2=3, 3=3.5, Index=$Particles.PartitionParticles.Partition, Num=::NewEnumerator2))
  ]
  particleUpdateScript [
   &ParticleState(Kill Particles When Lifetime Has Elapsed=false, Let Infinitely Lived Particles Die When Emitter Deactivates=true)
   &ScaleSpriteSize(Scale Sprite Size Mode=::NewEnumerator0, Uniform Scale Factor=$System.EngineOwnerScale)
   &DynamicMaterialParameters(
    Index 0 Param 1=1
    Index 0 Param 2=&MergeHelper_FloatSwitch(0=3, 1=3, 2=3, 3=5, Index=$Particles.PartitionParticles.Partition, Num=::NewEnumerator2)
    Module.Index0Float0Write=true
    Module.Index0Float1Write=true
    Module.Index0Float2Write=false
    Module.Index0Float3Write=false
   ) {module=DynamicMaterialParameters001}
   &ScaleColor(
    Color Scale:LinearColor=&MultiplyLinearColor(
     Linear Color=$System.TODColor
     Scale Factor=&MakeVector4FromVectorAndFloat("Vector (RGB)"=<1 1 1>, "Float (Alpha)"=&FN_FrontEndFloatModifier(Gameplay=1, Front End=0.6))
    )
    Scale Mode=3
   ) {module=ScaleColor001}
  ]
  rendererClasses [&NiagaraSpriteRendererProperties, &NiagaraMeshRendererProperties]
  renderers [
   {rendererClass=&NiagaraSpriteRendererProperties}
   {rendererClass=&NiagaraMeshRendererProperties, rendererIndex=1}
  ]
 }
 Light {bEnabled=false, rendererClasses [], renderers []}
}
summary {
 systemName=NS_MicStand_Outershell_Idle_FX
 emitters [
  {emitterName=CapsuleGlow, rendererClasses [&NiagaraSpriteRendererProperties, &NiagaraMeshRendererProperties]}
  {bEnabled=false, emitterName=Light, rendererClasses []}
 ]
 userVariables [
  User.bIsFrontEnd {type=NiagaraBool}
  User.bIsFrontEndPreview {type=NiagaraBool}
  User.IsInstrument {type=NiagaraFloat, defaultValue=0, description=""}
 ]
}
